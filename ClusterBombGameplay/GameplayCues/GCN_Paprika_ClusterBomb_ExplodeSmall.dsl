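; Burst gameplay cue for the small Paprika cluster bomb grenade's explosion, layering extra logic on the parent cue's effects.
; After the parent call, it casts the target to the small grenade and turns off decal spawning on the particle components if the main grenade is stuck to a moving platform.
; When the r.PhysicsField.EnableField console variable is 1, it also spawns a radial force field actor at the target's root transform.
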(fn OnBurst(EitherVFX-preferthisversion)
  (params (MyTarget :type Actor%20Object%20Reference) (Parameters :type Gameplay%20Cue%20Parameters%20Structure%20%28by%20ref%29) (ParticleComponents :type Array%20of%20FXSystem%20Component%20Object%20References) (AudioComponents :type Array%20of%20Audio%20Component%20Object%20References) (BurstCameraShakeInstance :type Camera%20Shake%20Base%20Object%20Reference) (BurstDecalInstance :type Decal%20Actor%20Object%20Reference))
  (|CallParentFunction MyTarget Parameters ParticleComponents AudioComponents BurstCameraShakeInstance BurstDecalInstance)
  ;[1 1 1] If the main grenade is attached to a moving platform do NOT spawn the decal
  (bind _asb_prj_paprika_cluster_bomb_grenade_small (Utilities|Casting|CastToB_Prj_Paprika_ClusterBomb_GrenadeSmall MyTarget)
    (:then
      (if (Variables|/ClusterBombGameplay/Gameplay/Actors/B_Prj_Paprika_ClusterBomb_GrenadeSmall.B_Prj_Paprika_ClusterBomb_GrenadeSmall_C|GetbMainGrenadeStuckToMovingPlatform _asb_prj_paprika_cluster_bomb_grenade_small)
        (for _array_element ParticleComponents
          (Effects|Components|ParticleSystem|SetBoolParameter _array_element "Spawn Decal" false))
        (else
          _))
      (bind _returnvalue (Development|GetConsoleVariableIntValue "r.PhysicsField.EnableField"))
      (if (== _returnvalue 1)
        (Game|SpawnActorfromClass 0 "/Game/Effects/Niagara/Blueprints/PhsyicsFields/FN_RadialForce.FN_RadialForce_C" (Transformation|GetWorldTransform (Variables|/Script/Engine.Actor|GetRootComponent MyTarget)) "Undefined" "MultiplyWithRoot" 0 0.2 0.66 8.5e+02 2.0 0.0 0)
        (return)))
    (:CastFailed
      _)))

(fn OnBurst(Niagara)
  (params (MyTarget :type Actor%20Object%20Reference) (Parameters :type Gameplay%20Cue%20Parameters%20Structure%20%28by%20ref%29) (ParticleComponents :type Array%20of%20Niagara%20Particle%20System%20Component%20Object%20References) (AudioComponents :type Array%20of%20Audio%20Component%20Object%20References) (BurstCameraShakeInstance :type Camera%20Shake%20Base%20Object%20Reference) (BurstDecalInstance :type Decal%20Actor%20Object%20Reference))
  (return))
